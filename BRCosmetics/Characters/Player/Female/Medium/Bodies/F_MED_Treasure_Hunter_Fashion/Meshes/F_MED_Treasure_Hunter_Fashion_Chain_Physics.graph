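# Physics asset for the chain on the Treasure Hunter Fashion cosmetic body for medium female player characters.
# It has two capsule bodies, a longer one on the dyn_chain_top bone and a small one on spine_05.
# A single constraint links the chain body to spine_05 with tight swing limits and a wider twist limit, restricting how far the chain rotates relative to the upper spine.
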PhysicsAsset /BRCosmetics/Characters/Player/Female/Medium/Bodies/F_MED_Treasure_Hunter_Fashion/Meshes/F_MED_Treasure_Hunter_Fashion_Chain_Physics
bodies {
 dyn_chain_top [{shapeType=Capsule, length=9.5, radius=1.5, shapeName=dyn_chain_top_capsule, center=<0 4.6251 -4.6251>, rotation=<rot 45 90 0>}]
 spine_05 [{shapeType=Capsule, length=1.2274, radius=1, center=<0.2458 0.2676 0.0263>, rotation=<rot 5.9142 -89.9995 -1.3486>}]
}
constraints [{bone1Name=dyn_chain_top, bone2Name=spine_05, swing1LimitDegrees=8, swing2LimitDegrees=1, twistLimitDegrees=25}]
