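# Physics asset for the Treasure Hunter Fashion medium female player body, setting up physics for two side pouches.
# Capsule bodies cover both pouches, their bracket bones, the pelvis and the left thigh.
# The bracket chain locks all angular motion to the pelvis, while the pouches hang from it with narrow swing and twist limits, the front one with one swing axis locked.
# The thigh capsule has no constraint, apparently serving as a collider.
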PhysicsAsset /BRCosmetics/Characters/Player/Female/Medium/Bodies/F_MED_Treasure_Hunter_Fashion/Meshes/F_MED_Treasure_Hunter_Fashion_Physics
bodies {
 dyn_pouch_back [{shapeType=Capsule, length=5.8122, radius=2.6205, center=<5.232 -2.306 -4.6552>, rotation=<rot 18.2736 -83.3703 -27.2102>}]
 dyn_pouch_front [{shapeType=Capsule, length=6.833, radius=2.9828, center=<3.257 3.0019 -4.988>, rotation=<rot -4.6214 -94.1227 -30.2258>}]
 dyn_side_bracket [{shapeType=Capsule, length=0.3179, radius=1.4225, center=<-0.1002 -0.0365 -0.2552>, rotation=<rot 5.2991 -89.9996 -21.2004>}]
 dyn_side_bracket_master [{shapeType=Capsule, length=0.505, radius=0.505, rotation=<rot 0 -90 0>}]
 pelvis [{shapeType=Capsule, length=2.1778, radius=1.5, center=<-0.7055 -0.0781 -0.0997>, rotation=<rot -12.4708 -79.804 -2.9056>}]
 thigh_l [{shapeType=Capsule, length=7.1705, radius=8.1593, center=<-1.3325 -1.4954 3.319>, rotation=<rot -0.9772 -91.2586 -108.1304>}]
}
constraints [
 {bone1Name=dyn_pouch_back, bone2Name=dyn_side_bracket, swing1LimitDegrees=20, swing2LimitDegrees=20, twistLimitDegrees=10}
 {
  bone1Name=dyn_pouch_front
  bone2Name=dyn_side_bracket
  swing1LimitDegrees=10
  swing1Motion=Locked
  swing2LimitDegrees=20
  twistLimitDegrees=10
 }
 {
  bone1Name=dyn_side_bracket_master
  bone2Name=pelvis
  swing1LimitDegrees=45
  swing1Motion=Locked
  swing2LimitDegrees=45
  swing2Motion=Locked
  twistLimitDegrees=45
  twistMotion=Locked
 }
 {
  bone1Name=dyn_side_bracket
  bone2Name=dyn_side_bracket_master
  swing1LimitDegrees=45
  swing1Motion=Locked
  swing2LimitDegrees=45
  swing2Motion=Locked
  twistLimitDegrees=45
  twistMotion=Locked
 }
]
