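# Post-process style animation blueprint for the female Treasure Hunter Fashion cosmetic body, adding secondary motion over the attached parent mesh's copied pose.
# A crouch-aware Control Rig feeds two pelvis-space RigidBody simulations, then a weighted constraint and AnimDynamics drive the right bracelet before a chain bone offset.
# Chain physics blend drops toward half as frame time rises.
# The crouch-driven side bracket bone transform is unconnected to the output.
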AnimBlueprint /BRCosmetics/Characters/Player/Female/Medium/Bodies/F_MED_Treasure_Hunter_Fashion/Meshes/F_MED_Treasure_Hunter_Fashion_AnimBp
target_skeleton=&/F_MED_Treasure_Hunter_Fashion_Skeleton
graphs [
 AnimGraph {nodes [
   "Transform (Modify) Bone - Bone: dyn_side_bracket_master"(Alpha=Get crouching alpha() -> (crouching alpha:double), Translation="0,-6,0", Rotation="10.000,-55.000,-5.000", Scale="1,1,1") {
    rotationMode=BMM_Additive
    rotationSpace=BCS_BoneSpace
    translationMode=BMM_Additive
    translationSpace=BCS_BoneSpace
    alphaBoolBlend {blendOption=Linear}
    alphaScaleBiasClamp {bInterpResult=true}
    boneToModify {boneName=dyn_side_bracket_master}
    rotation=<rot 10 -55 -5>
    translation {y=-6}
   }
   n1 = Control Rig!(Source=Copy Pose From Mesh() {bUseAttachedParent=true}, Alpha, crouching alpha:double=Get crouching alpha()) -> (Pose) {
    alphaBoolBlend {blendOption=Linear}
    controlRigAssetReference {blueprintRigClass=&/F_MED_Treasure_Hunter_Fashion_CtrlRig.F_MED_Treasure_Hunter_Fashion_CtrlRig_C}
   }
   n2 = RigidBody(ComponentPose=Local To Component(LocalPose=@n1)) {
    bClampLinearTranslationLimitToRefPose=true
    bOverrideWorldGravity=true
    overridePhysicsAsset=&/F_MED_Treasure_Hunter_Fashion_Physics
    simulationSpace=BaseBoneSpace
    alphaBoolBlend {blendOption=Linear}
    baseBoneRef {boneName=pelvis}
    componentAppliedLinearAccClamp=<4000 4000 4000>
    componentLinearAccScale=<0.5 0.5 0.5>
    componentLinearVelScale=<0.5 0.5 0.5>
    overrideWorldGravity {z=-980}
   }
   n3 = RigidBody(ComponentPose=@n2, Alpha=Get FpsAlpha() -> (FpsAlpha:double)) {
    bClampLinearTranslationLimitToRefPose=true
    bOverrideWorldGravity=true
    overridePhysicsAsset=&/F_MED_Treasure_Hunter_Fashion_Chain_Physics
    simulationSpace=BaseBoneSpace
    alphaBoolBlend {blendOption=Linear}
    baseBoneRef {boneName=pelvis}
    componentAppliedLinearAccClamp=<6000 8000 8000>
    componentLinearAccScale=<0.5 0.5 0.5>
    componentLinearVelScale=<0.5 0.5 0.5>
    overrideWorldGravity {z=-980}
   }
   n4 = "Constraint - dyn_r_bracelet_master (PPP)"!(ComponentPose=@n3, Alpha=1, ConstraintWeights_0=0.06, ConstraintWeights_1:float=0.0859, ConstraintWeights_2:float=0.8541) -> (Pose) {
    alphaBoolBlend {blendOption=Linear}
    boneToModify {boneName=dyn_r_bracelet_master}
    constraintSetup [{targetBone {boneName=dyn_r_bracelet}}, {targetBone {boneName=lowerarm_r}}, {targetBone {boneName=lowerarm_twist_01_r}}]
    constraintWeights [1,1,1]
   }
   n5 = Anim Dynamics - Bone: dyn_r_bracelet_master(ComponentPose=@n4, Alpha=1, ExternalForce="0,0,0") {
    angularDampingOverride=0.96
    angularSpringConstant=75
    bAngularSpring=true
    bGravityOverrideInSimSpace=true
    bLinearSpring=true
    bOverrideAngularDamping=true
    bOverrideLinearDamping=true
    simulationSpace=Actor
    alphaBoolBlend {blendOption=Linear}
    boundBone {boneName=dyn_r_bracelet_master}
    componentAppliedLinearAccClamp=<4000 4000 4000>
    componentLinearAccScale=<1 1 1>
    componentLinearVelScale=<0.5 0.5 0.5>
    physicsBodyDefinitions [{
      sphereCollisionRadius=0
      boundBone {boneName=dyn_r_bracelet_master}
      boxExtents=<5 18 18>
      constraintSetup {angularLimitsMax=<15 20 0>, angularLimitsMin=<-10 -20 0>, angularTarget=<1 0 0>}
      localJointOffset=<2 0 -4>
     }]
   }
   n6 = "Transform (Modify) Bone - Bone: dyn_chain_top"(ComponentPose=@n5, Alpha=1, Translation="0,0,0", Rotation="0,0,5", Scale="1,1,1") {
    rotationMode=BMM_Additive
    rotationSpace=BCS_BoneSpace
    alphaBoolBlend {blendOption=Linear}
    boneToModify {boneName=dyn_chain_top}
    rotation {roll=5}
   }
   Output Pose(Result=Component To Local(ComponentPose=@n6)) {layerGroup=None}
  ]}
 EventGraph {nodes [
   n1 = Event BlueprintUpdateAnimation()
   > Set FpsAlpha(FpsAlpha:double=MapRangeClamped(Value=@n1.DeltaTimeX, InRangeA=0.015, InRangeB=0.03, OutRangeA=1, OutRangeB=0.5)) -> (Output_Get:double)
   Event OnBeginCrouching()
   > Set crouching alpha(crouching alpha:double=1) -> (Output_Get:double)
   Event OnEndCrouching()
   > Set crouching alpha(crouching alpha=0)
   TryGetPawnOwner()
  ]}
]
variables [
 crouching alpha {type=number}
 FpsAlpha {type=number}
]
