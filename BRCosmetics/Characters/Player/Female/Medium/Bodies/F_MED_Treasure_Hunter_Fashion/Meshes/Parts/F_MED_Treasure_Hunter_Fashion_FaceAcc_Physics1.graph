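# Physics asset for the Treasure Hunter Fashion female cosmetic's FaceAcc part, setting up left and right ponytail physics.
# Each ponytail is a five-capsule chain hinged to the head, with tight root limits that loosen down the chain.
# Capsules on head, neck, upper spine, clavicles and upper arms form the body volume, mostly linked to spine_04 by free-angle constraints.
# The right chain's second joint is stiffer than the left's.
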PhysicsAsset /BRCosmetics/Characters/Player/Female/Medium/Bodies/F_MED_Treasure_Hunter_Fashion/Meshes/Parts/F_MED_Treasure_Hunter_Fashion_FaceAcc_Physics1
bodies {
 clavicle_l [{
   shapeType=Capsule
   length=2.9581
   radius=4.5897
   shapeName=clavicle_l_capsule
   center=<6.536 -0.4481 -0.7094>
   rotation=<rot 16.9211 87.287 -90>
  }]
 clavicle_r [{
   shapeType=Capsule
   length=3.8659
   radius=4.8987
   shapeName=clavicle_r_capsule
   center=<-6.4298 0.2933 1.4023>
   rotation=<rot 3.2384 98.0696 -90>
  }]
 dyn_ponytail_l_01 [{shapeType=Capsule, length=8.125, radius=2.25, shapeName=dyn_hair_l_Front_01_capsule, center=<2.5 0 0>, rotation=<rot 90 0 0>}]
 dyn_ponytail_l_02 [{shapeType=Capsule, length=5.375, radius=2.25, shapeName=dyn_hair_l_Front_01_capsule, center=<4 2.5 0>, rotation=<rot 90 0 -40.0003>}]
 dyn_ponytail_l_03 [{shapeType=Capsule, length=2.3313, radius=2.4436, shapeName=dyn_hair_l_Front_02_capsule, center=<2 1 0>, rotation=<rot 90 180 149.9998>}]
 dyn_ponytail_l_04 [{shapeType=Capsule, length=1.375, radius=2.4614, shapeName=dyn_hair_l_Front_03_capsule, center=<1.5 1 0>, rotation=<rot 90 6.3402 -18.6601>}]
 dyn_ponytail_l_05 [{shapeType=Capsule, length=2.5, radius=2, shapeName=dyn_hair_l_Front_04_capsule, center=<2.5 1.5 0>, rotation=<rot 90 180 149.9998>}]
 dyn_ponytail_r_01 [{shapeType=Capsule, length=8.125, radius=2.25, shapeName=dyn_hair_l_Front_01_capsule, center=<2.5 0 0>, rotation=<rot 90 0 0>}]
 dyn_ponytail_r_02 [{shapeType=Capsule, length=5.375, radius=2.25, shapeName=dyn_hair_l_Front_01_capsule, center=<4 2.5 0>, rotation=<rot 90 0 -40.0003>}]
 dyn_ponytail_r_03 [{shapeType=Capsule, length=2.7063, radius=2.4436, shapeName=dyn_hair_l_Front_02_capsule, center=<2 1 0>, rotation=<rot 90 180 149.9998>}]
 dyn_ponytail_r_04 [{shapeType=Capsule, length=1.375, radius=2.0864, shapeName=dyn_hair_l_Front_03_capsule, center=<1.5 1 0>, rotation=<rot 90 6.3402 -18.6601>}]
 dyn_ponytail_r_05 [{shapeType=Capsule, length=3.375, radius=1.75, shapeName=dyn_hair_r_Front_04_capsule, center=<2.5 1.5 0>, rotation=<rot 90 180 149.9998>}]
 head [{shapeType=Capsule, length=3.35, radius=5.2337, center=<1.5273 5.4619 -0.0165>, rotation=<rot -89.5256 -89.7443 90.2186>}]
 neck_02 [{shapeType=Capsule, length=11.4843, radius=3.9291, center=<-2.0439 0 0>, rotation=<rot 0 -89.9999 -89.9996>}]
 spine_04 [{shapeType=Capsule, length=3.0316, radius=10.6376, shapeName=spine_04_capsule, center=<0.7655 1.5405 0>, rotation=<rot 0 -92.6847 0>}]
 spine_05 [
  {shapeType=Capsule, length=7.6911, radius=8.1451, shapeName=spine_05_capsule, center=<1.7745 -0.215 -0.1551>}
  {
   shapeType=Capsule
   length=8
   radius=3.7
   shapeName=spine_05_capsule
   center=<3.2258 4.5613 -5.4722>
   rotation=<rot -77.6635 113.3999 -158.0877>
  }
  {shapeType=Capsule, length=8, radius=3.7, shapeName=spine_05_capsule, center=<3.2028 4.518 5.4302>, rotation=<rot -75 -50.0002 5>}
  {shapeType=Capsule, length=11.25, radius=2.625, shapeName=spine_05_capsule, center=<0.4285 8.4531 0>}
 ]
 upperarm_l [{shapeType=Capsule, length=22.991, radius=3.3862, center=<11.4355 2.0132 0.8986>, rotation=<rot 90 0 0>}]
 upperarm_r [{shapeType=Capsule, length=16.3973, radius=3.2924, center=<-11.4355 2.0132 0>, rotation=<rot 90 0 0>}]
}
constraints [
 {bone1Name=neck_02, bone2Name=spine_05, swing1LimitDegrees=45, swing2LimitDegrees=45, twistLimitDegrees=45}
 {bone1Name=dyn_ponytail_l_01, bone2Name=head, swing1LimitDegrees=10, swing2LimitDegrees=10, twistLimitDegrees=2}
 {bone1Name=dyn_ponytail_r_01, bone2Name=head, swing1LimitDegrees=10, swing2LimitDegrees=10, twistLimitDegrees=2}
 {
  bone1Name=spine_05
  bone2Name=spine_04
  swing1LimitDegrees=45
  swing1Motion=Free
  swing2LimitDegrees=45
  swing2Motion=Free
  twistLimitDegrees=45
  twistMotion=Free
 }
 {
  bone1Name=clavicle_l
  bone2Name=spine_04
  swing1LimitDegrees=45
  swing1Motion=Free
  swing2LimitDegrees=45
  swing2Motion=Free
  twistLimitDegrees=45
  twistMotion=Free
 }
 {
  bone1Name=upperarm_l
  bone2Name=spine_04
  swing1LimitDegrees=45
  swing1Motion=Free
  swing2LimitDegrees=45
  swing2Motion=Free
  twistLimitDegrees=45
  twistMotion=Free
 }
 {
  bone1Name=clavicle_r
  bone2Name=spine_04
  swing1LimitDegrees=45
  swing1Motion=Free
  swing2LimitDegrees=45
  swing2Motion=Free
  twistLimitDegrees=45
  twistMotion=Free
 }
 {
  bone1Name=upperarm_r
  bone2Name=spine_04
  swing1LimitDegrees=45
  swing1Motion=Free
  swing2LimitDegrees=45
  swing2Motion=Free
  twistLimitDegrees=45
  twistMotion=Free
 }
 {
  bone1Name=head
  bone2Name=spine_04
  swing1LimitDegrees=45
  swing1Motion=Free
  swing2LimitDegrees=45
  swing2Motion=Free
  twistLimitDegrees=45
  twistMotion=Free
 }
 {bone1Name=dyn_ponytail_r_05, bone2Name=dyn_ponytail_r_04, swing1LimitDegrees=45, swing2LimitDegrees=35, twistLimitDegrees=15}
 {bone1Name=dyn_ponytail_r_02, bone2Name=dyn_ponytail_r_01, swing1LimitDegrees=35, swing2LimitDegrees=20, twistLimitDegrees=5}
 {bone1Name=dyn_ponytail_r_03, bone2Name=dyn_ponytail_r_02, swing1LimitDegrees=45, swing2LimitDegrees=35, twistLimitDegrees=15}
 {bone1Name=dyn_ponytail_r_04, bone2Name=dyn_ponytail_r_03, swing1LimitDegrees=45, swing2LimitDegrees=35, twistLimitDegrees=15}
 {bone1Name=dyn_ponytail_l_05, bone2Name=dyn_ponytail_l_04, swing1LimitDegrees=45, swing2LimitDegrees=35, twistLimitDegrees=15}
 {bone1Name=dyn_ponytail_l_02, bone2Name=dyn_ponytail_l_01, swing1LimitDegrees=45, swing2LimitDegrees=35, twistLimitDegrees=15}
 {bone1Name=dyn_ponytail_l_03, bone2Name=dyn_ponytail_l_02, swing1LimitDegrees=45, swing2LimitDegrees=35, twistLimitDegrees=15}
 {bone1Name=dyn_ponytail_l_04, bone2Name=dyn_ponytail_l_03, swing1LimitDegrees=45, swing2LimitDegrees=35, twistLimitDegrees=15}
]
