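# Physics asset for the Treasure Hunter Fashion medium female player body's ADS variant, built for secondary motion of two ponytails.
# Each ponytail is six capsules chained from the head, with angular limits that keep twist very tight and allow wider swing mid-chain.
# Head, neck, clavicle, arm and compound upper-spine shapes cover the upper body.
# Only neck-to-spine is also constrained; the two small chain-accessory capsules have no constraints.
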PhysicsAsset /Game/Characters/Player/Female/Medium/Bodies/F_MED_Treasure_Hunter_Fashion/Meshes/Parts/F_MED_Treasure_Hunter_Fashion_ADS_Phys
bodies {
 clavicle_l [{shapeType=Capsule, length=8.0801, radius=6.4382, center=<2.4642 -1.2295 -2.2169>, rotation=<rot 90 0 0>}]
 clavicle_r [{shapeType=Capsule, length=7.5801, radius=6.9382, center=<-5.2708 1.6148 3.2538>, rotation=<rot 90 0 0>}]
 dyn_chain [{shapeType=Capsule, length=1.25, radius=1.005, rotation=<rot -39.9999 -89.9999 0>}]
 dyn_chain_top [{shapeType=Capsule, length=2.75, radius=1.5, rotation=<rot 0 0 -45.0008>}]
 dyn_ponytail_l_01 [{shapeType=Capsule, length=5.5018, radius=2.7728, center=<3.6733 -0.9653 0>, rotation=<rot 90 0 0>}]
 dyn_ponytail_l_02 [{shapeType=Capsule, length=6.5956, radius=1.9603, center=<4.3618 2.4165 -0.0823>, rotation=<rot 2.8082 132.4473 -87.2825>}]
 dyn_ponytail_l_03 [{shapeType=Capsule, length=4.7461, radius=1.0973, center=<3.5127 1.4534 -0.1077>, rotation=<rot -0.3239 -63.1853 -88.1622>}]
 dyn_ponytail_l_04 [{shapeType=Capsule, length=2.0319, radius=1.069, center=<1.3102 0.5197 -0.176>, rotation=<rot -17.3846 -65.9191 -89.2456>}]
 dyn_ponytail_l_05 [{shapeType=Capsule, length=2.7672, radius=1.0122, center=<0.4761 0.9091 -0.019>, rotation=<rot -0.3783 116.1364 -91.0394>}]
 dyn_ponytail_l_06 [{shapeType=Capsule, length=2.1598, radius=0.8235, center=<-0.2672 0.5445 -0.004>, rotation=<rot 0.2871 -63.0808 -86.7169>}]
 dyn_ponytail_r_01 [{shapeType=Capsule, length=6.7831, radius=1.7416, center=<3.5201 0 0>, rotation=<rot 90 0 0>}]
 dyn_ponytail_r_02 [{shapeType=Capsule, length=6.7831, radius=1.7416, center=<4.5063 2.1309 0.0099>, rotation=<rot 0.435 -48.2573 -84.499>}]
 dyn_ponytail_r_03 [{shapeType=Capsule, length=4.6271, radius=1.4858, center=<2.3156 1.2032 -0.3575>, rotation=<rot -0.8355 -61.3657 -90.3991>}]
 dyn_ponytail_r_04 [{shapeType=Capsule, length=1.6282, radius=2.1226, center=<1.4991 1.0624 -0.2007>, rotation=<rot 90 0 -25.0103>}]
 dyn_ponytail_r_05 [{shapeType=Capsule, length=2.8366, radius=1.173, center=<1.9856 1.0931 -0.1701>, rotation=<rot 0.7566 -66.2818 -90.5039>}]
 dyn_ponytail_r_06 [{shapeType=Capsule, length=1.4601, radius=1.0102, center=<0.9699 0.4522 0.4718>, rotation=<rot 79.9977 -154.9961 -179.9999>}]
 head [{shapeType=Capsule, length=3.35, radius=5.2337, center=<1.5439 3.4489 0.0001>, rotation=<rot -89.5256 -89.7443 90.2186>}]
 lowerarm_l [{shapeType=Capsule, length=19.5938, radius=3.1613, center=<7 0 0>, rotation=<rot 0 -89.9999 -94.9996>}]
 lowerarm_r [{shapeType=Capsule, length=19.5, radius=4.505, center=<-7 0 0>, rotation=<rot -0.0001 -89.9999 89.9998>}]
 neck_02 [{shapeType=Capsule, length=11.4843, radius=3.9291, center=<-2.0439 -0.0001 0>, rotation=<rot 0 -89.9999 -89.9996>}]
 spine_05 [
  {shapeType=Sphere, radius=8.9062, center=<6.8713 -6.5175 8.9975>}
  {shapeType=Capsule, length=15.2245, radius=7.0796, center=<-2.6109 3.6168 -7.2056>, rotation=<rot -86.1669 67.5009 -87.3425>}
  {shapeType=Capsule, length=18.1933, radius=6.2359, center=<-2.8534 3.8263 8.7689>, rotation=<rot 86.9754 72.4146 102.3014>}
  {shapeType=Capsule, length=25.1875, radius=4.8125, center=<6.3632 -0.0207 0>}
  {shapeType=Box, extentX=18.5671, extentY=8.9811, extentZ=22.7407, center=<-2.4539 5.7822 -0.1051>, rotation=<rot 0 -29.9975 0>}
 ]
 upperarm_l [{shapeType=Capsule, length=27.741, radius=3.0424, center=<14.0998 -0.0002 0.8986>, rotation=<rot 90 0 0>}]
 upperarm_r [{shapeType=Capsule, length=20.6473, radius=3.2612, center=<-11.4355 0 0>, rotation=<rot 90 0 0>}]
}
constraints [
 {bone1Name=neck_02, bone2Name=spine_05, swing1LimitDegrees=45, swing2LimitDegrees=45, twistLimitDegrees=45}
 {bone1Name=dyn_ponytail_l_02, bone2Name=dyn_ponytail_l_01, swing1LimitDegrees=30, swing2LimitDegrees=30, twistLimitDegrees=2}
 {bone1Name=dyn_ponytail_l_03, bone2Name=dyn_ponytail_l_02, swing1LimitDegrees=50, swing2LimitDegrees=40, twistLimitDegrees=2}
 {bone1Name=dyn_ponytail_l_04, bone2Name=dyn_ponytail_l_03, swing1LimitDegrees=50, swing2LimitDegrees=50, twistLimitDegrees=5}
 {bone1Name=dyn_ponytail_r_02, bone2Name=dyn_ponytail_r_01, swing1LimitDegrees=30, swing2LimitDegrees=30, twistLimitDegrees=2}
 {bone1Name=dyn_ponytail_r_03, bone2Name=dyn_ponytail_r_02, swing1LimitDegrees=50, swing2LimitDegrees=40, twistLimitDegrees=2}
 {bone1Name=dyn_ponytail_r_04, bone2Name=dyn_ponytail_r_03, swing1LimitDegrees=50, swing2LimitDegrees=50, twistLimitDegrees=5}
 {bone1Name=dyn_ponytail_l_01, bone2Name=head, swing1LimitDegrees=20, swing2LimitDegrees=20, twistLimitDegrees=2}
 {bone1Name=dyn_ponytail_r_01, bone2Name=head, swing1LimitDegrees=20, swing2LimitDegrees=20, twistLimitDegrees=2}
 {bone1Name=dyn_ponytail_r_05, bone2Name=dyn_ponytail_r_04, swing1LimitDegrees=40, swing2LimitDegrees=40, twistLimitDegrees=5}
 {bone1Name=dyn_ponytail_r_06, bone2Name=dyn_ponytail_r_05, swing1LimitDegrees=20, swing2LimitDegrees=20, twistLimitDegrees=5}
 {bone1Name=dyn_ponytail_l_05, bone2Name=dyn_ponytail_l_04, swing1LimitDegrees=40, swing2LimitDegrees=40, twistLimitDegrees=5}
 {bone1Name=dyn_ponytail_l_06, bone2Name=dyn_ponytail_l_05, swing1LimitDegrees=20, swing2LimitDegrees=20, twistLimitDegrees=2}
]
