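# Post-process style AnimBP for the female Treasure Hunter Fashion face accessory, animating its ponytail bones over the pose copied from the parent mesh.
# The output path caches that pose with small bone offsets, applies LOD1-weighted corrective shifts, then runs RigidBody simulation based on spine_05.
# The event graph sets LOD and emote weights, plus head-versus-spine rotation alphas when rigid body is disabled.
# Procedural look-at, constraint, trail and alternate physics branches do not reach the output.
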AnimBlueprint /BRCosmetics/Characters/Player/Female/Medium/Bodies/F_MED_Treasure_Hunter_Fashion/Meshes/Parts/F_MED_TreasureHunterFash_FaceAcc_AnimBp
target_skeleton=/Game/Characters/Player/Female/Medium/Bodies/F_MED_Treasure_Hunter_Fashion/Meshes/Parts/F_MED_Treasure_Hunter_Fashion_FaceAcc_Skeleton
graphs [
 AnimGraph {nodes [
   n1 = Get pitch-right_head_rotation_alpha() -> (pitch-right_head_rotation_alpha:double)
   n2 = Look At - Bone: dyn_ponytail_l_01(ComponentPose=Local To Component(LocalPose=Use cached pose 'PreSym'()), Alpha=1) {
    alphaBoolBlend {blendOption=Linear}
    boneToModify {boneName=dyn_ponytail_l_01}
    lookAtLocation=<3.5381 0.4928 0>
    lookAtTarget {boneReference {boneName=clavicle_l}}
    lookAt_Axis {axis=<1 0>}
   }
   n3 = Look At - Bone: dyn_ponytail_r_01(ComponentPose=@n2, Alpha=1) {
    alphaBoolBlend {blendOption=Linear}
    boneToModify {boneName=dyn_ponytail_r_01}
    lookAtLocation=<-3.2021 0 0>
    lookAtTarget {boneReference {boneName=clavicle_r}}
    lookAt_Axis {axis=<1 0>}
   }
   n4 = Look At - Bone: dyn_ponytail_l_02(ComponentPose=@n3, Alpha=1) {
    alphaBoolBlend {blendOption=Linear}
    boneToModify {boneName=dyn_ponytail_l_02}
    lookAtLocation=<3.7089 1.5238 0>
    lookAtTarget {boneReference {boneName=clavicle_l}}
    lookAt_Axis {axis=<1 0>}
   }
   n5 = Look At - Bone: dyn_ponytail_r_02(ComponentPose=@n4, Alpha=1) {
    alphaBoolBlend {blendOption=Linear}
    boneToModify {boneName=dyn_ponytail_r_02}
    lookAtLocation=<-4.1765 -1.2111>
    lookAtTarget {boneReference {boneName=clavicle_r}}
    lookAt_Axis {axis=<1 0>}
   }
   n6 = Get left_head_rotation_alpha() -> (left_head_rotation_alpha:double)
   n7 = "Transform (Modify) Bone - Bone: dyn_ponytail_r_02"(ComponentPose=@n5, Alpha=@n6, Translation="0,0,0", Rotation="-0.031,15,-0.001", Scale="1,1,1") {
    rotationMode=BMM_Additive
    rotationSpace=BCS_BoneSpace
    translationMode=BMM_Additive
    alphaBoolBlend {blendOption=Linear}
    boneToModify {boneName=dyn_ponytail_r_02}
    rotation=<rot -0.031 15 -0.001>
   }
   n8 = "Transform (Modify) Bone - Bone: dyn_ponytail_r_03"(ComponentPose=@n7, Alpha=@n6, Translation="0,0,0", Rotation="-0.031,15,-0.001", Scale="1,1,1") {
    rotationMode=BMM_Additive
    rotationSpace=BCS_BoneSpace
    translationMode=BMM_Additive
    alphaBoolBlend {blendOption=Linear}
    boneToModify {boneName=dyn_ponytail_r_03}
    rotation=<rot -0.031 15 -0.001>
   }
   n9 = Get right_head_rotation_alpha() -> (right_head_rotation_alpha:double)
   n10 = "Transform (Modify) Bone - Bone: dyn_ponytail_r_03"(ComponentPose=@n8, Alpha=@n9, Translation="0,0,0", Rotation="0,-20,0", Scale="1,1,1") {
    rotationMode=BMM_Additive
    rotationSpace=BCS_BoneSpace
    translationMode=BMM_Additive
    translationSpace=BCS_BoneSpace
    alphaBoolBlend {blendOption=Linear}
    boneToModify {boneName=dyn_ponytail_r_03}
    rotation {yaw=-20}
   }
   n11 = "Transform (Modify) Bone - Bone: dyn_ponytail_l_03"(ComponentPose=@n10, Alpha=@n9, Translation="-0.420,1.492,0.158", Rotation="1.136,29.720,-3.560", Scale="1,1,1") {
    rotationMode=BMM_Additive
    rotationSpace=BCS_BoneSpace
    translationMode=BMM_Additive
    translationSpace=BCS_BoneSpace
    alphaBoolBlend {blendOption=Linear}
    alphaScaleBiasClamp {bInterpResult=true, interpSpeedDecreasing=2, interpSpeedIncreasing=2}
    boneToModify {boneName=dyn_ponytail_l_03}
    rotation=<rot 1.136 29.72 -3.56>
    translation=<-0.42 1.492 0.158>
   }
   n12 = Get targeting-top_head_rotation_alpha() -> (targeting-top_head_rotation_alpha:double)
   n13 = "Transform (Modify) Bone - Bone: dyn_ponytail_r_01"(ComponentPose=@n11, Alpha=@n12, Translation="0,0,0", Rotation="0.000,-5,-0.000", Scale="1,1,1") {
    rotationMode=BMM_Additive
    rotationSpace=BCS_BoneSpace
    alphaBoolBlend {blendOption=Linear}
    boneToModify {boneName=dyn_ponytail_r_01}
    rotation {yaw=-5}
   }
   n14 = "Transform (Modify) Bone - Bone: dyn_ponytail_r_04"(ComponentPose=@n13, Alpha=@n12, Translation="0,0,0", Rotation="0,-10,0", Scale="1,1,1") {
    rotationMode=BMM_Additive
    rotationSpace=BCS_BoneSpace
    alphaBoolBlend {blendOption=Linear}
    boneToModify {boneName=dyn_ponytail_r_04}
    rotation {yaw=-10}
   }
   n15 = "Transform (Modify) Bone - Bone: dyn_ponytail_r_05"(ComponentPose=@n14, Alpha=@n12, Translation="0,0,0", Rotation="0,-20,0.980", Scale="1,1,1") {
    rotationMode=BMM_Additive
    rotationSpace=BCS_BoneSpace
    alphaBoolBlend {blendOption=Linear}
    boneToModify {boneName=dyn_ponytail_r_05}
    rotation {roll=0.98, yaw=-20}
   }
   n16 = "Transform (Modify) Bone - Bone: dyn_ponytail_r_01"(ComponentPose=@n15, Alpha=@n1, Translation="0,0,0", Rotation="0,10,-0.000", Scale="1,1,1") {
    rotationMode=BMM_Additive
    rotationSpace=BCS_BoneSpace
    alphaBoolBlend {blendOption=Linear}
    boneToModify {boneName=dyn_ponytail_r_01}
    rotation {yaw=10}
   }
   n17 = Get targeting-down_head_rotation_alpha() -> (targeting-down_head_rotation_alpha:double)
   n18 = "Transform (Modify) Bone - Bone: dyn_ponytail_r_03"(ComponentPose=@n16, Alpha=@n17, Translation="0,0,0", Rotation="0,30,0", Scale="1,1,1") {
    rotationMode=BMM_Additive
    rotationSpace=BCS_BoneSpace
    alphaBoolBlend {blendOption=Linear}
    boneToModify {boneName=dyn_ponytail_r_03}
    rotation {yaw=30}
   }
   n19 = "Transform (Modify) Bone - Bone: dyn_ponytail_l_02"(ComponentPose=@n18, Alpha=Get right_head_rotation_alpha(), Translation="0,0,0", Rotation="-0.000,8,0.000", Scale="1,1,1") {
    rotationMode=BMM_Additive
    rotationSpace=BCS_BoneSpace
    alphaBoolBlend {blendOption=Linear}
    boneToModify {boneName=dyn_ponytail_l_02}
    rotation {yaw=8}
   }
   n20 = Get top-left_head_rotation_alpha() -> (top-left_head_rotation_alpha:double)
   n21 = "Transform (Modify) Bone - Bone: dyn_ponytail_l_03"(ComponentPose=@n19, Alpha=@n20, Translation="0,0,0", Rotation="0,-20,0", Scale="1,1,1") {
    rotationMode=BMM_Additive
    rotationSpace=BCS_BoneSpace
    alphaBoolBlend {blendOption=Linear}
    boneToModify {boneName=dyn_ponytail_l_03}
    rotation {yaw=-20}
   }
   n22 = "Transform (Modify) Bone - Bone: dyn_ponytail_r_04"(ComponentPose=@n21, Alpha=@n20, Translation="-1.5,0.5,0", Rotation="0,20,0", Scale="1,1,1") {
    rotationMode=BMM_Additive
    rotationSpace=BCS_BoneSpace
    translationMode=BMM_Additive
    translationSpace=BCS_BoneSpace
    alphaBoolBlend {blendOption=Linear}
    boneToModify {boneName=dyn_ponytail_r_04}
    rotation {yaw=20}
    translation=<-1.5 0.5>
   }
   n23 = Get top_head_rotation_alpha() -> (top_head_rotation_alpha:double)
   n24 = "Transform (Modify) Bone - Bone: dyn_ponytail_l_01"(ComponentPose=@n22, Alpha=@n23, Translation="0,0,0", Rotation="0,5,0", Scale="1,1,1") {
    rotationMode=BMM_Additive
    rotationSpace=BCS_BoneSpace
    alphaBoolBlend {blendOption=Linear}
    boneToModify {boneName=dyn_ponytail_l_01}
    rotation {yaw=5}
   }
   n25 = "Transform (Modify) Bone - Bone: dyn_ponytail_l_02"(ComponentPose=@n24, Alpha=@n23, Translation="0,0,0", Rotation="0,-7,0", Scale="1,1,1") {
    rotationMode=BMM_Additive
    rotationSpace=BCS_BoneSpace
    alphaBoolBlend {blendOption=Linear}
    boneToModify {boneName=dyn_ponytail_l_02}
    rotation {yaw=-7}
   }
   n26 = "Transform (Modify) Bone - Bone: dyn_ponytail_l_04"(ComponentPose=@n25, Alpha=@n23, Translation="0,0,0", Rotation="0,-10,0", Scale="1,1,1") {
    rotationMode=BMM_Additive
    rotationSpace=BCS_BoneSpace
    alphaBoolBlend {blendOption=Linear}
    boneToModify {boneName=dyn_ponytail_l_04}
    rotation {yaw=-10}
   }
   n27 = "Transform (Modify) Bone - Bone: dyn_ponytail_r_01"(ComponentPose=@n26, Alpha=@n23, Translation="0,0,0", Rotation="0,5,0", Scale="1,1,1") {
    rotationMode=BMM_Additive
    rotationSpace=BCS_BoneSpace
    alphaBoolBlend {blendOption=Linear}
    boneToModify {boneName=dyn_ponytail_r_01}
    rotation {yaw=5}
   }
   n28 = "Transform (Modify) Bone - Bone: dyn_ponytail_r_02"(ComponentPose=@n27, Alpha=@n23, Translation="0,0,0", Rotation="0,-7,0", Scale="1,1,1") {
    rotationMode=BMM_Additive
    rotationSpace=BCS_BoneSpace
    alphaBoolBlend {blendOption=Linear}
    boneToModify {boneName=dyn_ponytail_r_02}
    rotation {yaw=-7}
   }
   n29 = "Transform (Modify) Bone - Bone: dyn_ponytail_r_04"(ComponentPose=@n28, Alpha=@n23, Translation="0,0,0", Rotation="0,-10,0", Scale="1,1,1") {
    rotationMode=BMM_Additive
    rotationSpace=BCS_BoneSpace
    alphaBoolBlend {blendOption=Linear}
    boneToModify {boneName=dyn_ponytail_r_04}
    rotation {yaw=-10}
   }
   n30 = "Transform (Modify) Bone - Bone: dyn_ponytail_l_01"(ComponentPose=@n29, Alpha=@n1, Translation="-0.347,1.902,1.098", Rotation="0.001,-8.000,0.000", Scale="1,1,1") {
    rotationMode=BMM_Additive
    rotationSpace=BCS_BoneSpace
    translationMode=BMM_Additive
    alphaBoolBlend {blendOption=Linear}
    boneToModify {boneName=dyn_ponytail_l_01}
    rotation=<rot 0.001 -7.9999 0>
    translation=<-0.347 1.902 1.098>
   }
   n31 = "Transform (Modify) Bone - Bone: dyn_ponytail_l_03"(ComponentPose=@n30, Alpha=@n17, Translation="0,0,0", Rotation="0,30,0", Scale="1,1,1") {
    rotationMode=BMM_Additive
    rotationSpace=BCS_BoneSpace
    alphaBoolBlend {blendOption=Linear}
    boneToModify {boneName=dyn_ponytail_l_03}
    rotation {yaw=30}
   }
   n32 = Look At - Bone: dyn_ponytail_l_03(ComponentPose=@n31, Alpha=1) {
    alphaBoolBlend {blendOption=Linear}
    boneToModify {boneName=dyn_ponytail_l_03}
    lookAtLocation=<0.6301 10.9428 -2.6954>
    lookAtTarget {boneReference {boneName=spine_03}}
    lookAt_Axis {axis=<1 0>}
   }
   n33 = Look At - Bone: dyn_ponytail_r_03(ComponentPose=@n32, Alpha=1) {
    alphaBoolBlend {blendOption=Linear}
    boneToModify {boneName=dyn_ponytail_r_03}
    lookAtLocation=<0.6294 10.3 2.8651>
    lookAtTarget {boneReference {boneName=spine_03}}
    lookAt_Axis {axis=<1 0>}
   }
   n34 = "Constraint - dyn_ponytail_r_05 (P)"(ComponentPose=@n33, Alpha=0.5, ConstraintWeights_0=1) {
    alphaBoolBlend {blendOption=Linear}
    boneToModify {boneName=dyn_ponytail_r_05}
    constraintSetup [{targetBone {boneName=dyn_ponytail_r_04}}]
    constraintWeights [1]
   }
   n35 = "Constraint - dyn_ponytail_l_05 (P)"(ComponentPose=@n34, Alpha=0.5, ConstraintWeights_0=1) {
    alphaBoolBlend {blendOption=Linear}
    boneToModify {boneName=dyn_ponytail_l_05}
    constraintSetup [{targetBone {boneName=dyn_ponytail_l_04}}]
    constraintWeights [1]
   }
   Blend Poses by bool(BlendPose_1=Component To Local(ComponentPose=@n35), bActiveValue=Get bRigidBodyEnabled()) {blendType=Linear}
   n36 = "Transform (Modify) Bone - Bone: dyn_ponytail_l_04"(Alpha=1, Translation="0,0.75,0", Rotation="0,0,0", Scale="1,1,1") {translationMode=BMM_Additive, alphaBoolBlend {blendOption=Linear}, boneToModify {boneName=dyn_ponytail_l_04}, translation {y=0.75}}
   n37 = "Transform (Modify) Bone - Bone: dyn_ponytail_l_03"(ComponentPose=@n36, Alpha=1, Translation="0,0.5,0", Rotation="0,0,0", Scale="1,1,1") {translationMode=BMM_Additive, alphaBoolBlend {blendOption=Linear}, boneToModify {boneName=dyn_ponytail_l_03}, translation {y=0.5}}
   Blend Poses by bool(BlendPose_0=Component To Local(ComponentPose=@n37), BlendTime_0=0, BlendTime_1=0, bActiveValue=Get bIsTargeting()) {blendType=Linear}
   n38 = Look At - Bone: dyn_ponytail_r_02(Alpha=1) {
    alphaBoolBlend {blendOption=Linear}
    boneToModify {boneName=dyn_ponytail_r_02}
    lookAtLocation=<-4.1765 -1.2111>
    lookAtTarget {boneReference {boneName=clavicle_r}}
    lookAt_Axis {axis=<1 0>}
   }
   n39 = Look At - Bone: dyn_ponytail_l_02(ComponentPose=@n38, Alpha=1) {
    alphaBoolBlend {blendOption=Linear}
    boneToModify {boneName=dyn_ponytail_l_02}
    lookAtLocation=<4.1765 1.2111>
    lookAtTarget {boneReference {boneName=clavicle_l}}
    lookAt_Axis {axis=<1 0>}
   }
   n40 = "Constraint - dyn_ponytail_r_02 (PP)"!(ComponentPose=@n39, Alpha=1, ConstraintWeights_0=0.6, ConstraintWeights_1:float=0.4) -> (Pose) {
    alphaBoolBlend {blendOption=Linear}
    boneToModify {boneName=dyn_ponytail_r_02}
    constraintSetup [{targetBone {boneName=spine_05}}, {targetBone {boneName=head}}]
    constraintWeights [1,1]
   }
   n41 = "Constraint - dyn_ponytail_l_02 (PP)"!(ComponentPose=@n40, Alpha=1, ConstraintWeights_0=0.6, ConstraintWeights_1:float=0.4) -> (Pose) {
    alphaBoolBlend {blendOption=Linear}
    boneToModify {boneName=dyn_ponytail_l_02}
    constraintSetup [{targetBone {boneName=spine_05}}, {targetBone {boneName=head}}]
    constraintWeights [1,1]
   }
   n42 = "Constraint - dyn_ponytail_r_03 (PP)"!(ComponentPose=@n41, Alpha=1, ConstraintWeights_0=0.9, ConstraintWeights_1:float=0.1) -> (Pose) {
    alphaBoolBlend {blendOption=Linear}
    boneToModify {boneName=dyn_ponytail_r_03}
    constraintSetup [{targetBone {boneName=spine_05}}, {targetBone {boneName=head}}]
    constraintWeights [1,1]
   }
   "Constraint - dyn_ponytail_l_03 (PP)"!(ComponentPose=@n42, Alpha=1, ConstraintWeights_0=0.9, ConstraintWeights_1:float=0.1) -> (Pose) {
    alphaBoolBlend {blendOption=Linear}
    boneToModify {boneName=dyn_ponytail_l_03}
    constraintSetup [{targetBone {boneName=spine_05}}, {targetBone {boneName=head}}]
    constraintWeights [1,1]
   }
   n43 = "Constraint - dyn_ponytail_r_02 (PP)"(Alpha=1, ConstraintWeights_0=0.7, ConstraintWeights_1=0.3) {
    alphaBoolBlend {blendOption=Linear}
    boneToModify {boneName=dyn_ponytail_r_02}
    constraintSetup [{targetBone {boneName=spine_05}}, {targetBone {boneName=head}}]
    constraintWeights [1,1]
   }
   n44 = "Constraint - dyn_ponytail_l_02 (PP)"(ComponentPose=@n43, Alpha=1, ConstraintWeights_0=0.7, ConstraintWeights_1=0.3) {
    alphaBoolBlend {blendOption=Linear}
    boneToModify {boneName=dyn_ponytail_l_02}
    constraintSetup [{targetBone {boneName=spine_05}}, {targetBone {boneName=head}}]
    constraintWeights [1,1]
   }
   n45 = "Constraint - dyn_ponytail_r_03 (PP)"(ComponentPose=@n44, Alpha=1, ConstraintWeights_0=0.9, ConstraintWeights_1=0.1) {
    alphaBoolBlend {blendOption=Linear}
    boneToModify {boneName=dyn_ponytail_r_03}
    constraintSetup [{targetBone {boneName=spine_05}}, {targetBone {boneName=head}}]
    constraintWeights [1,1]
   }
   "Constraint - dyn_ponytail_l_03 (PP)"(ComponentPose=@n45, Alpha=1, ConstraintWeights_0=0.9, ConstraintWeights_1=0.1) {
    alphaBoolBlend {blendOption=Linear}
    boneToModify {boneName=dyn_ponytail_l_03}
    constraintSetup [{targetBone {boneName=spine_05}}, {targetBone {boneName=head}}]
    constraintWeights [1,1]
   }
   Look At - Bone: dyn_ponytail_l_01(Alpha=1) {
    alphaBoolBlend {blendOption=Linear}
    boneToModify {boneName=dyn_ponytail_l_01}
    lookAtLocation {x=3.2021}
    lookAtTarget {boneReference {boneName=clavicle_l}}
    lookAt_Axis {axis=<1 0>}
   }
   Look At - Bone: dyn_ponytail_r_01(Alpha=1) {
    alphaBoolBlend {blendOption=Linear}
    boneToModify {boneName=dyn_ponytail_r_01}
    lookAtLocation {x=-3.2021}
    lookAtTarget {boneReference {boneName=clavicle_r}}
    lookAt_Axis {axis=<1 0>}
   }
   n46 = Get LOD1 Selected() -> (LOD1 Selected:double)
   "Transform (Modify) Bone - Bone: corrective_bone_l"(Alpha=@n46, Translation="0,0,-2", Rotation="0,0,0", Scale="1,1,1") {translationMode=BMM_Additive, alphaBoolBlend {blendOption=Linear}, boneToModify {boneName=corrective_bone_l}, translation {z=-2}}
   RigidBody(ComponentPose=Local To Component(LocalPose=Use cached pose 'PreSym'())) {
    bClampLinearTranslationLimitToRefPose=true
    bOverrideWorldGravity=true
    overridePhysicsAsset=/Game/Characters/Player/Female/Medium/Bodies/F_MED_Treasure_Hunter_Fashion/Meshes/Parts/F_MED_Treasure_Hunter_Fashion_ADS_Phys
    simulationSpace=BaseBoneSpace
    alphaBoolBlend {blendOption=Linear}
    baseBoneRef {boneName=head}
    componentAppliedLinearAccClamp=<600 600 600>
    componentLinearAccScale=<0.6 0.6 0.6>
    componentLinearVelScale=<0.6 0.6 0.6>
    overrideWorldGravity {z=-980}
   }
   n47 = Get LOD1 Selected()
   n48 = "Transform (Modify) Bone - Bone: corrective_bone_l"(
    ComponentPose=Local To Component(LocalPose=Use cached pose 'PreSym'())
    Alpha=@n47
    Translation="0.0,0.0,-1"
    Rotation="0,0,0"
    Scale="1,1,1"
   ) {translationMode=BMM_Additive, alphaBoolBlend {blendOption=Linear}, boneToModify {boneName=corrective_bone_l}, translation {z=-1}}
   n49 = "Transform (Modify) Bone - Bone: corrective_bone_r"(ComponentPose=@n48, Alpha=@n47, Translation="0.0,0.0,-1", Rotation="0,0,0", Scale="1,1,1") {translationMode=BMM_Additive, alphaBoolBlend {blendOption=Linear}, boneToModify {boneName=corrective_bone_r}, translation {z=-1}}
   n50 = RigidBody!(ComponentPose=@n49, Alpha, OverrideWorldGravity:Vector="0,0,-980", ExternalForce) -> (Pose) {
    bClampLinearTranslationLimitToRefPose=true
    bOverrideWorldGravity=true
    overridePhysicsAsset=&/F_MED_Treasure_Hunter_Fashion_FaceAcc_Physics1
    simulationSpace=BaseBoneSpace
    alphaBoolBlend {blendOption=Linear}
    baseBoneRef {boneName=spine_05}
    componentAppliedLinearAccClamp=<1000 1000 1000>
    componentLinearAccScale=<0.2 0.2 0.3>
    componentLinearVelScale=<0.1 0.1 0.2>
   }
   Output Pose(Result=Component To Local(ComponentPose=@n50)) {layerGroup=None}
   n51 = "Transform (Modify) Bone - Bone: corrective_bone_r"(
    ComponentPose=Local To Component(LocalPose=Copy Pose From Mesh() {bCopyCurves=true, bUseAttachedParent=true})
    Alpha=@n46
    Translation="-1.5,0,0"
    Rotation="0,0,0"
    Scale="1,1,1"
   ) {translationMode=BMM_Additive, alphaBoolBlend {blendOption=Linear}, boneToModify {boneName=corrective_bone_r}, translation {x=-1.5}}
   n52 = "Transform (Modify) Bone - Bone: dyn_ponytail_r_04"(ComponentPose=@n51, Alpha=1, Translation="0,0.75,0", Rotation="0,0,0", Scale="1,1,1") {translationMode=BMM_Additive, alphaBoolBlend {blendOption=Linear}, boneToModify {boneName=dyn_ponytail_r_04}, translation {y=0.75}}
   Save cached pose 'PreSym'(Pose=Component To Local(ComponentPose=@n52)) {cacheName=PreSym}
   n53 = Trail controller - Bone: dyn_ponytail_r_05(Alpha=Get emote_playing() -> (emote_playing:double)) {
    bLimitRotation=true
    bUsePlanarLimit=true
    chainLength=4
    stretchLimit=0.01
    alphaScaleBiasClamp {outRange {max=0.8, min=0.5}}
    planarLimits [
     {drivingBone {boneName=spine_05}, planeTransform {location=<0 12 0>, rotation=<rot 0 -25 90>, scale=<0.2 0.5 0.5>}}
     {drivingBone {boneName=upperarm_r}, planeTransform {location=<-14 0 0>, rotation=<rot 0 0 -45>, scale=<0.35 0.2 0.2>}}
    ]
    rotationLimits [
     {limitMax=<5 5 40>, limitMin=<-5 -5 0>}
     {limitMax=<5 5 40>, limitMin=<-5 -5 0>}
     {limitMax=<5 10 40>, limitMin=<-5 -10 0>}
     {limitMax=<5 10 20>, limitMin=<-5 -10 0>}
    ]
    rotationOffsets [<0 0 0>, <0 0 0>, <0 0 0>, <0 0 0>]
    trailBone {boneName=dyn_ponytail_r_05}
    trailRelaxationSpeed {editorCurveData {keys [<at 0 10 RCIM_Linear>,<at 0.0479 5.5175 RCIM_Linear>]}}
   }
   Trail controller - Bone: dyn_ponytail_l_05(ComponentPose=@n53, Alpha=Get emote_playing()) {
    bLimitRotation=true
    bUsePlanarLimit=true
    chainLength=4
    stretchLimit=0.01
    planarLimits [{drivingBone {boneName=spine_05}, planeTransform {location=<0 12 0>, rotation=<rot 0 -25 90.0001>, scale=<0.2 0.5 0.5>}}]
    rotationLimits [
     {limitMax=<5 5 40>, limitMin=<-5 -5 0>}
     {limitMax=<5 5 40>, limitMin=<-5 -5 0>}
     {limitMax=<5 10 40>, limitMin=<-5 -10 0>}
     {limitMax=<5 10 20>, limitMin=<-5 -10 0>}
    ]
    rotationOffsets [<0 0 0>, <0 0 0>, <0 0 0>, <0 0 0>]
    trailBone {boneName=dyn_ponytail_l_05}
    trailRelaxationSpeed {editorCurveData {keys [<at 0.0775 19.914 RCIM_Linear>,<at 0.1014 7.8746 RCIM_Linear>]}}
   }
   Get Braid_Alpha() -> (Braid_Alpha:double)
  ]}
 EventGraph {nodes [
   Event BlueprintUpdateAnimation()
   > n1 = Sequence()
   n2 = Get bIsTargeting()
   n3 = Get bIsSkydiveDiveMode()
   n4 = Get bIsSkydiveFloating()
   n5 = Get bIsSkyDiving()
   n6 = Get bIsParachuteOpened()
   > Set Braid_Alpha(Braid_Alpha:double=SelectFloat(A=0.25, B=1, bPickA=OR Boolean!(A=@n5, B=@n6, C:bool=@n4, D:bool=@n3, E:bool=@n2) -> (ReturnValue))) -> (Output_Get:double)
   n7 = "Equal (Integer)"(A=Get CurrentLOD(), B=1)
   Set LOD1 Selected(execute=@n1.then_1, LOD1 Selected:double="To Float (Boolean)"(InBool=@n7)) -> (Output_Get:double)
   > Branch(Condition=NOT Boolean(A=Get bRigidBodyEnabled()))
   n8 = MapRangeClamped(Value="To Float (Boolean)"(InBool=Get bIsPlayingEmote()), InRangeA=0, InRangeB=1, OutRangeA=0.5, OutRangeB=0.8)
   n9 = AND Boolean(A=@n7, B=Get bIsInFrontEnd())
   > Set emote_playing(emote_playing:double=SelectFloat(A=0.2, B=SelectFloat(A=0, B=@n8, bPickA=@n9), bPickA=@n7)) -> (Output_Get:double)
   n10 = Get character_skeletal_mesh() -> (character_skeletal_mesh:Skeletal Mesh Component Object Reference)
   n11 = GetAlphaFromReferenceTransformDelta(SkeletalMeshComponent=@n10, BoneName=head, BaseName=spine_05, DeltaComponentType=EulerX, InputRange=(X=0,Y=1), TwistAxis="1,0,0")
   > Set right_head_rotation_alpha(right_head_rotation_alpha:double=MapRangeClamped(Value=@n11.RawOutput, InRangeA=-20, InRangeB=0, OutRangeA=1, OutRangeB=0)) -> (Output_Get:double)
   n12 = MapRangeClamped(Value=@n11.RawOutput, InRangeA=0, InRangeB=40)
   > Set left_head_rotation_alpha(left_head_rotation_alpha:double=@n12) -> (Output_Get:double)
   n13 = GetAlphaFromReferenceTransformDelta(SkeletalMeshComponent=@n10, BoneName=head, BaseName=spine_05, InputRange=(X=0,Y=1), TwistAxis="1,0,0")
   n14 = MapRangeClamped(Value=@n13.RawOutput, InRangeA=0, InRangeB=-45)
   > Set top-left_head_rotation_alpha(top-left_head_rotation_alpha:double=float * float(A=@n12, B=@n14) {is=title K2Node_PromotableOperator}) -> (Output_Get:double)
   > Set top_head_rotation_alpha(top_head_rotation_alpha:double=@n14) -> (Output_Get:double)
   > Set targeting-top_head_rotation_alpha(targeting-top_head_rotation_alpha:double=float * float(A=@n14)) -> (Output_Get:double)
   > Set targeting-down_head_rotation_alpha(targeting-down_head_rotation_alpha:double=MapRangeClamped(Value=@n13.RawOutput, InRangeA=7, InRangeB=25)) -> (Output_Get:double)
   > Set pitch-right_head_rotation_alpha(pitch-right_head_rotation_alpha:double=MapRangeClamped(
     Value=GetAlphaFromReferenceTransformDelta(SkeletalMeshComponent=@n10, BoneName=head, BaseName=spine_05, DeltaComponentType=EulerY, InputRange=(X=0,Y=1), TwistAxis="1,0,0").RawOutput
     InRangeA=0
     InRangeB=33
    )) -> (Output_Get:double)
   Event OnCharacterCustomizationLoaded()
   > Set character_skeletal_mesh(character_skeletal_mesh:Skeletal Mesh Component Object Reference=GetOwnerSkeletalMeshComponent()) -> (Output_Get:Skeletal Mesh Component Object Reference)
   n15 = Branch(Condition=Get bIsTargeting())
   > Do Once(Reset=@n17.then)
   > n16 = ResetDynamics(InTeleportType=ResetPhysics)
   Do Once(execute=@n15.else, Reset=@n16.then)
   > n17 = ResetDynamics(InTeleportType=ResetPhysics)
   Get bIsTargeting()
  ]}
]
variables [
 Braid_Alpha {type=number}
 character_skeletal_mesh {type=SkeletalMeshComponent}
 emote_playing {type=number}
 left_head_rotation_alpha {type=number}
 LOD1 Selected {type=number}
 pitch-right_head_rotation_alpha {type=number}
 right_head_rotation_alpha {type=number}
 targeting-down_head_rotation_alpha {type=number}
 targeting-top_head_rotation_alpha {type=number}
 top-left_head_rotation_alpha {type=number}
 top_head_rotation_alpha {type=number}
]
