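# Physics asset for the FaceAcc mesh part of the female medium Treasure Hunter Fashion cosmetic, covering left and right ponytail physics.
# Each ponytail is a six-capsule chain constrained to the head, with very tight twist and swing limits that widen mid-chain and narrow toward the tip.
# Head, neck, clavicle, upper-arm and compound upper-spine shapes provide body collision volumes.
# A dyn_chain_top capsule has no constraint, and neck_02 is constrained to spine_05.
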PhysicsAsset /BRCosmetics/Characters/Player/Female/Medium/Bodies/F_MED_Treasure_Hunter_Fashion/Meshes/Parts/F_MED_Treasure_Hunter_Fashion_FaceAcc_Physics
bodies {
 clavicle_l [{shapeType=Capsule, length=3.3613, radius=6.4382, center=<3.8684 -1.3803 0.01>, rotation=<rot 90 0 0>}]
 clavicle_r [{shapeType=Capsule, length=7.5801, radius=6.9382, center=<-5.2708 1.6148 3.2538>, rotation=<rot 90 0 0>}]
 dyn_chain_top [{shapeType=Capsule, length=2.75, radius=1.5, rotation=<rot 0 0 -45.0008>}]
 dyn_ponytail_l_01 [{shapeType=Capsule, length=5.5018, radius=2.7728, center=<3.6733 -0.9653 0>, rotation=<rot 90 0 0>}]
 dyn_ponytail_l_02 [{shapeType=Capsule, length=4.8456, radius=2, center=<3.9822 2.8315 -0.0546>, rotation=<rot 3.2097 123.2196 -87.7711>}]
 dyn_ponytail_l_03 [{shapeType=Capsule, length=1.7461, radius=2, center=<2.4897 0.8557 -0.0702>, rotation=<rot -0.5018 -57.5852 -88.2026>}]
 dyn_ponytail_l_04 [{shapeType=Capsule, length=0.9381, radius=1.75, center=<0.943 1.1481 0.0516>, rotation=<rot -17.3822 -60.6791 -90.812>}]
 dyn_ponytail_l_05 [{shapeType=Capsule, length=0.7359, radius=1.5, center=<0.9177 2.1181 -0.0892>, rotation=<rot -0.3495 117.7163 -91.0494>}]
 dyn_ponytail_l_06 [{shapeType=Capsule, length=0.3786, radius=1.1047, center=<-0.5337 1.5446 -0.112>, rotation=<rot 0.2226 -61.9582 -86.7119>}]
 dyn_ponytail_r_01 [{shapeType=Capsule, length=6.7831, radius=1.7416, center=<3.5201 0 0>, rotation=<rot 90 0 0>}]
 dyn_ponytail_r_02 [{shapeType=Capsule, length=3.7831, radius=2.25, center=<3.9175 2.7907 0.0032>, rotation=<rot 0.2521 -46.3604 -84.4876>}]
 dyn_ponytail_r_03 [{shapeType=Capsule, length=1.8771, radius=2.25, center=<2.037 1.6111 -0.4547>, rotation=<rot -0.7958 -56.1534 -87.6118>}]
 dyn_ponytail_r_04 [{shapeType=Capsule, length=0.3782, radius=2.1226, center=<0.7912 2.545 -0.2007>, rotation=<rot 88.3236 31.8373 0>}]
 dyn_ponytail_r_05 [{shapeType=Capsule, length=0.8366, radius=1.5, center=<1.2375 2.7747 -0.3511>, rotation=<rot 0.7989 -61.1168 -88.7901>}]
 dyn_ponytail_r_06 [{shapeType=Capsule, length=0.5851, radius=1.1352, center=<0.1984 2.1907 0.2696>, rotation=<rot 79.9977 -154.9961 -179.9999>}]
 head [{shapeType=Capsule, length=3.35, radius=5.2337, center=<1.5439 3.4489 0.0001>, rotation=<rot -89.5256 -89.7443 90.2186>}]
 neck_02 [{shapeType=Capsule, length=11.4843, radius=3.9291, center=<-2.0439 -0.0001 0>, rotation=<rot 0 -89.9999 -89.9996>}]
 spine_05 [
  {shapeType=Sphere, radius=8.9062, center=<6.8713 -6.5175 8.9975>}
  {shapeType=Capsule, length=16.0683, radius=6.2359, center=<-2.379 2.4377 -5.8907>, rotation=<rot -86.1715 67.4831 -97.3247>}
  {shapeType=Capsule, length=16.0683, radius=6.2359, center=<-1.53 3.9602 8.2507>, rotation=<rot 86.9742 72.4135 102.3005>}
  {shapeType=Box, extentX=15.5671, extentY=11, extentZ=20.2407, center=<-0.2641 5.0109 0.2882>, rotation=<rot 0 -35.1874 0>}
  {shapeType=Box, extentX=20.875, extentY=7.5, extentZ=48.625, center=<14.3619 -9.8342 1.4241>}
 ]
 upperarm_l [{shapeType=Capsule, length=27.741, radius=3.8862, center=<11.4355 0 0.8986>, rotation=<rot 90 0 0>}]
 upperarm_r [{shapeType=Capsule, length=21.1473, radius=3.7924, center=<-11.4355 0 0>, rotation=<rot 90 0 0>}]
}
constraints [
 {bone1Name=neck_02, bone2Name=spine_05, swing1LimitDegrees=45, swing2LimitDegrees=45, twistLimitDegrees=45}
 {bone1Name=dyn_ponytail_l_02, bone2Name=dyn_ponytail_l_01, swing1LimitDegrees=45, swing2LimitDegrees=45, twistLimitDegrees=2}
 {bone1Name=dyn_ponytail_l_03, bone2Name=dyn_ponytail_l_02, swing1LimitDegrees=50, swing2LimitDegrees=40, twistLimitDegrees=2}
 {bone1Name=dyn_ponytail_l_04, bone2Name=dyn_ponytail_l_03, swing1LimitDegrees=50, swing2LimitDegrees=50, twistLimitDegrees=2}
 {bone1Name=dyn_ponytail_r_02, bone2Name=dyn_ponytail_r_01, swing1LimitDegrees=45, swing2LimitDegrees=45, twistLimitDegrees=2}
 {bone1Name=dyn_ponytail_r_03, bone2Name=dyn_ponytail_r_02, swing1LimitDegrees=50, swing2LimitDegrees=40, twistLimitDegrees=2}
 {bone1Name=dyn_ponytail_r_04, bone2Name=dyn_ponytail_r_03, swing1LimitDegrees=50, swing2LimitDegrees=50, twistLimitDegrees=2}
 {bone1Name=dyn_ponytail_l_01, bone2Name=head, swing1LimitDegrees=20, swing2LimitDegrees=20, twistLimitDegrees=2}
 {bone1Name=dyn_ponytail_r_01, bone2Name=head, swing1LimitDegrees=20, swing2LimitDegrees=20, twistLimitDegrees=2}
 {bone1Name=dyn_ponytail_r_05, bone2Name=dyn_ponytail_r_04, swing1LimitDegrees=40, swing2LimitDegrees=40, twistLimitDegrees=2}
 {bone1Name=dyn_ponytail_r_06, bone2Name=dyn_ponytail_r_05, swing1LimitDegrees=30, swing2LimitDegrees=30, twistLimitDegrees=2}
 {bone1Name=dyn_ponytail_l_05, bone2Name=dyn_ponytail_l_04, swing1LimitDegrees=40, swing2LimitDegrees=40, twistLimitDegrees=2}
 {bone1Name=dyn_ponytail_l_06, bone2Name=dyn_ponytail_l_05, swing1LimitDegrees=30, swing2LimitDegrees=30, twistLimitDegrees=2}
]
